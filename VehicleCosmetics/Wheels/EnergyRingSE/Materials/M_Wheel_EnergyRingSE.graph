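# Translucent unlit material for the EnergyRingSE vehicle wheel, a glowing animated energy effect that passes through the shared damage and de-rez functions.
# Panned and rotated ray, wisp, noise and bolt textures combine into three named flow masks.
# Opacity blends in a ring mask and normal-mapped Fresnel term.
# Emissive blends flow and overlap colors built from custom or secondary colors, with switchable hue shifts and a view-angle tint.
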Material /VehicleCosmetics/Wheels/EnergyRingSE/Materials/M_Wheel_EnergyRingSE
graph {
 properties {
  bAutomaticallySetUsageInEditor=false
  bUseMaterialAttributes=true
  bUsedWithMeshDeformer=true
  bUsedWithSkeletalMesh=true
  blendMode=BLEND_Translucent
  shadingModel=MSM_Unlit
  translucencyLightingMode=TLM_SurfacePerPixelLighting
  usedShadingModels=Unlit
 }
 nodes [
  n1 = NamedRerouteUsage()
  n2 = NamedRerouteUsage()
  MaterialFunctionCall(In=MaterialFunctionCall(Material=SetMaterialAttributes!(MaterialAttributes, Emissive Color=@n1, Opacity=@n2) {attributeSetTypes [#1, #2]}) {materialFunction=&MF_BaseVehicle_Damage}.Result) {materialFunction=&MF_VehicleDeRez_Transition}
  n3 = Panner(Coordinate=TextureCoordinate() {uTiling=2, vTiling=5}) {speedY=-0.5}
  n4 = Constant() {r=1}
  n5 = TextureObject() {texture=/VehicleCosmetics/Wheels/_SharedTextures/FX/RaysGradient_RGB_T}
  n6 = Panner(Coordinate=TextureCoordinate() {vTiling=5}) {speedY=1}
  n7 = Constant() {r=1}
  n8 = Add(
   A=TextureSample(UVs=Rotator(Coordinate=@n3, Time=@n4) {speed=1.57}, Tex=@n5) {samplerSource=SSM_Wrap_WorldGroupSettings, texture=&RaysGradient_RGB_T}.B
   B=TextureSample(UVs=Rotator(Coordinate=@n6, Time=@n7) {speed=1.57}, Tex=@n5) {samplerSource=SSM_Wrap_WorldGroupSettings, texture=&RaysGradient_RGB_T}.G
  )
  n9 = TextureCoordinate()
  n10 = TextureSample(UVs=Rotator(Coordinate=Panner(Coordinate=@n9) {speedY=-1}, Time=Constant() {r=1}) {speed=1.57}, Tex=@n5) {samplerSource=SSM_Wrap_WorldGroupSettings, texture=&RaysGradient_RGB_T}
  n11 = Multiply(A=@n8, B=@n10.G)
  n12 = TextureCoordinate() {uTiling=3, vTiling=5}
  n13 = TextureSample(UVs=Rotator(Coordinate=Panner(Coordinate=@n12) {speedY=-0.3}, Time=Constant() {r=1}) {speed=1.57}) {
   samplerSource=SSM_Wrap_WorldGroupSettings
   samplerType=SAMPLERTYPE_LinearColor
   texture=/VehicleCosmetics/Wheels/_SharedTextures/FX/WispTrail_RGB_T
  }
  n14 = Add(A=Multiply(A=@n8, B=@n13.R), B=@n10.G)
  n15 = Panner(Coordinate=TextureCoordinate() {uTiling=5, vTiling=5}) {speedY=-0.12}
  n16 = TextureObject() {texture=/VehicleCosmetics/Wheels/_SharedTextures/FX/Noise_Spots_RGB_T_nomip}
  n17 = TextureSample(UVs=Panner(Coordinate=TextureCoordinate() {uTiling=0.5, vTiling=0.5}) {speedY=-0.05}, Tex=@n16) {samplerSource=SSM_Wrap_WorldGroupSettings, texture=&Noise_Spots_RGB_T_nomip}
  n18 = Multiply(A=Multiply(
    A=TextureSample(UVs=@n15, Tex=@n16) {samplerSource=SSM_Wrap_WorldGroupSettings, texture=&Noise_Spots_RGB_T_nomip}.G
    B=Multiply(A=@n17.R) {constB=3}
   ) {constB=3}) {constB=4}
  NamedRerouteDeclaration(None=Add(A=@n11, B=Add(A=@n14, B=@n18))) {name=Mask 1, nodeColor=<rgba 0.0078 1 0 1>}
  n19 = Panner(Coordinate=TextureCoordinate()) {speedY=-0.05}
  n20 = Constant() {r=1}
  n21 = Multiply(A=TextureSample(UVs=Rotator(Coordinate=@n19, Time=@n20) {speed=1.57}, Tex=@n5) {samplerSource=SSM_Wrap_WorldGroupSettings, texture=&RaysGradient_RGB_T}.B) {constB=2}
  n22 = Rotator(Coordinate=Panner(Coordinate=TextureCoordinate() {vTiling=2}) {speedY=0.08}, Time=Constant() {r=1}) {speed=1.57}
  n23 = Add(
   A=@n21
   B=Multiply(
    A=@n21
    B=TextureSample(UVs=@n22) {samplerSource=SSM_Wrap_WorldGroupSettings, texture=/VehicleCosmetics/Wheels/_SharedTextures/FX/DistortNoise_02_RGB_T}.B
   )
  )
  n24 = Panner(Coordinate=TextureCoordinate() {vTiling=-1}) {speedY=0.6}
  n25 = Constant() {r=1}
  n26 = TextureObject() {texture=/VehicleCosmetics/Wheels/_SharedTextures/FX/BoltCone_RGB_T}
  n27 = TextureCoordinate()
  n28 = Multiply(A=TextureSample(UVs=Panner(Coordinate=@n27) {speedY=-0.3}, Tex=@n5) {samplerSource=SSM_Wrap_WorldGroupSettings, texture=&RaysGradient_RGB_T}.G) {constB=2}
  n29 = Multiply(
   A=TextureSample(UVs=Rotator(Coordinate=@n24, Time=@n25) {speed=0}, Tex=@n26) {samplerSource=SSM_Wrap_WorldGroupSettings, texture=&BoltCone_RGB_T}.B
   B=@n28
  )
  n30 = TextureCoordinate() {vTiling=-3}
  n31 = Multiply(
   A=TextureSample(UVs=Panner(Coordinate=@n30) {speedY=0.5}, Tex=@n26) {samplerSource=SSM_Wrap_WorldGroupSettings, texture=&BoltCone_RGB_T}.R
   B=@n28
  )
  n32 = TextureCoordinate() {vTiling=-2}
  n33 = TextureSample(UVs=Rotator(Coordinate=Panner(Coordinate=@n32) {speedY=0.6}, Time=Constant() {r=1}) {speed=0}, Tex=@n26) {samplerSource=SSM_Wrap_WorldGroupSettings, texture=&BoltCone_RGB_T}
  n34 = Add(A=@n11, B=Add(A=Add(A=@n29, B=@n31), B=Multiply(A=@n33.G) {constB=2}))
  NamedRerouteDeclaration(None=Add(A=@n23, B=@n34)) {name=Mask 2, nodeColor=<rgba 0.0078 1 0 1>}
  NamedRerouteDeclaration(None=Multiply(A=@n23, B=@n34)) {name=Mask 3, nodeColor=<rgba 0.0078 1 0 1>}
  n35 = TextureSample() {samplerSource=SSM_Wrap_WorldGroupSettings, texture=/VehicleCosmetics/Wheels/_SharedTextures/FX/ringmask2}
  n36 = Multiply(A=@n35.RGB, B=NamedRerouteUsage())
  n37 = TextureSample() {
   samplerSource=SSM_Wrap_WorldGroupSettings
   samplerType=SAMPLERTYPE_Normal
   texture=/VehicleCosmetics/Wheels/EnergyRingSE/Textures/T_EnergyringSE_Tire_N
  }
  n38 = DotProduct(A=@n37.RGB, B=CameraVectorWS())
  n39 = Constant() {r=0.9}
  n40 = NamedRerouteDeclaration(None=Multiply(
    A=MaterialFunctionCall!(Normal=@n38, Exponent=@n39) {materialFunction=/VehicleCosmetics/Wheels/Cola_Holographic/Materials/MF_Fresnel_Classic}.Result
    B=ScalarParameter() {defaultValue=1, parameterName=Fresnel Emissive Multiplier}
   )) {name=Fresnel, nodeColor=<rgba 0.5703 1 0 1>}
  n41 = Multiply(A=Add(A=Multiply(A=@n36, B=@n40), B=@n40), B=@n35.RGB)
  n42 = NamedRerouteUsage()
  n43 = NamedRerouteUsage()
  NamedRerouteDeclaration(None=Clamp(None=LinearInterpolate(A=@n41, B=@n42, Alpha=@n43))) {name=Opacity, nodeColor=<rgba 0.9688 1 0 1>}
  n44 = VectorParameter() {parameterName=Custom Color, defaultValue=<rgba 0 0.4735 0.3613 1>}
  n45 = ScalarParameter() {defaultValue=-0.15, parameterName=Hue Shift}
  n46 = MaterialFunctionCall!(Color=@n44.RGB, Angle=OneMinus(None=@n45)) {materialFunction=/VehicleCosmetics/Wheels/Concept_Inverted/Materials/MF_HUE_Shift}
  n47 = Transform(None=CameraVectorWS()) {transformSourceType=TRANSFORMSOURCE_World, transformType=TRANSFORM_Tangent}
  n48 = DotProduct(A=@n47, B=Constant() {r=0.1})
  n49 = DotProduct(A=OneMinus(None=@n47), B=Constant() {r=0.2})
  n50 = StaticSwitchParameter(True=Add(A=Multiply(A=@n46.Result, B=@n48), B=Multiply(A=@n49, B=@n44.RGB)), False=Constant()) {defaultValue=true, parameterName=Multi Emissive Blend}
  n51 = VectorParameter() {parameterName=Secondary Color, defaultValue=<rgba 0 0.4735 0.3613 1>}
  n52 = Add(A=@n50, B=Multiply(A=StaticSwitchParameter(True=@n44.RGB, False=@n51.RGB) {parameterName=Tread Hue Shift}, B=NamedRerouteUsage()))
  n53 = Add(
   A=Multiply(A=@n50, B=NamedRerouteUsage())
   B=Multiply(A=Clamp(None=@n52) {maxDefault=2.5, minDefault=-50}, B=ScalarParameter() {defaultValue=1, parameterName=Flow Emissive Multiplier})
  )
  n54 = StaticSwitchParameter(
   True=MaterialFunctionCall(Color=@n44.RGB, Angle=@n45) {materialFunction=&MF_HUE_Shift}.Result
   False=@n51.RGB
  ) {parameterName=Tread Hue Shift}
  n55 = NamedRerouteUsage()
  n56 = Multiply(A=Add(A=@n50, B=Multiply(A=@n54, B=@n55)), B=ScalarParameter() {defaultValue=1, parameterName=Overlap Emissive Multiplier})
  n57 = NamedRerouteUsage()
  NamedRerouteDeclaration(None=Clamp(None=LinearInterpolate(A=@n53, B=@n56, Alpha=@n57))) {name=Emissive, nodeColor=<rgba 1 0 0.8906 1>}
  NamedRerouteDeclaration(None=VertexColor().R) {name=VC_R, nodeColor=<rgba 1 0 0.4687 1>}
 ]
}
