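# Material for the center of the EnergyRingSE vehicle wheel, a fully rough Default Lit surface with an animated emissive glow.
# A normal map feeds classic Fresnel terms at two exponents, which combine with mask texture channels and a custom color to build the clamped emissive.
# World position offset moves vertices along two opposed sine/cosine paths with separate speeds, blended by vertex color red and scaled by an intensity parameter.
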Material /VehicleCosmetics/Wheels/EnergyRingSE/Materials/M_Wheel_EnergyRingSE_Center
graph {
 properties {
  bAutomaticallySetUsageInEditor=false
  bUseMaterialAttributes=true
  bUsedWithMeshDeformer=true
  bUsedWithSkeletalMesh=true
  usedShadingModels=Default Lit
 }
 nodes [
  n1 = TextureSampleParameter2D() {parameterName=Normal, samplerType=SAMPLERTYPE_Normal, texture=/VehicleCosmetics/Wheels/EnergyRingSE/Textures/T_EnergyringSE_Rim_N}
  NamedRerouteDeclaration(None=@n1.RGB) {name=Normal, nodeColor=<rgba 0 0.2734 1 1>}
  n2 = CameraVectorWS()
  NamedRerouteDeclaration(None=MaterialFunctionCall!(Normal=DotProduct(A=@n1.RGB, B=@n2), Exponent=Constant() {r=5}) {materialFunction=/VehicleCosmetics/Wheels/Cola_Holographic/Materials/MF_Fresnel_Classic}.Result) {name=Fresnel 5, nodeColor=<rgba 0.5937 1 0 1>}
  NamedRerouteDeclaration(None=MaterialFunctionCall(Normal=DotProduct(A=@n1.RGB, B=@n2), Exponent=Constant() {r=2}) {materialFunction=&MF_Fresnel_Classic}.Result) {name=Fresnel 2, nodeColor=<rgba 0.5938 1 0 1>}
  n3 = TextureSample() {samplerType=SAMPLERTYPE_LinearColor, texture=/VehicleCosmetics/Wheels/EnergyRingSE/Textures/T_EnergyringSE_Rim_M}
  n4 = NamedRerouteUsage()
  n5 = Multiply(A=Multiply(A=@n3.R, B=@n4)) {constB=0.5}
  n6 = VectorParameter() {parameterName=Custom Color, defaultValue=<rgba 0.0704 0 1 1>}
  n7 = Multiply(A=@n6.RGB) {constB=2}
  n8 = Multiply(A=@n6.RGB, B=ScalarParameter() {defaultValue=4, parameterName=Rim Emissive Mulitplier}) {constB=2}
  n9 = OneMinus(None=@n4)
  n10 = LinearInterpolate(B=Add(A=Multiply(A=@n5, B=@n7) {constB=0.5}, B=Multiply(A=@n8, B=@n9) {constB=0.5}), Alpha=@n3.R)
  n11 = Multiply(A=Multiply(A=@n8, B=NamedRerouteUsage()) {constB=0.5}, B=@n3.B)
  NamedRerouteDeclaration(None=Clamp(None=Add(A=@n10, B=@n11))) {name=Emissive, nodeColor=<rgba 0 1 0.7656 1>}
  NamedRerouteDeclaration(None=LinearInterpolate(Alpha=@n3.R) {constB=0}) {name=BC, nodeColor=<rgba 1 0.4453 0 1>}
  n12 = Multiply(A=Time(), B=ScalarParameter() {defaultValue=0.2, parameterName=Speed})
  n13 = Sine(None=@n12)
  n14 = Cosine(None=@n12)
  n15 = Multiply(A=AppendVector(A=AppendVector(A=@n13, B=@n14), B=@n14), B=Constant3Vector() {constant=<rgba 0.5 0.2 0.5 0>})
  n16 = ScalarParameter() {defaultValue=1, parameterName=Intensity}
  n17 = Multiply(A=Time(), B=ScalarParameter() {defaultValue=0.2, parameterName=Speed 2})
  n18 = Sine(None=@n17)
  n19 = Cosine(None=@n17)
  n20 = Multiply(A=AppendVector(A=AppendVector(A=@n18, B=@n19), B=@n19), B=Constant3Vector() {constant=<rgba -0.5 -0.2 -0.5 0>})
  NamedRerouteDeclaration(None=LinearInterpolate(A=Multiply(A=@n15, B=@n16), B=Multiply(A=@n20, B=@n16), Alpha=VertexColor().R)) {name=WPO, nodeColor=<rgba 0.125 0 1 1>}
  n21 = Constant()
  SetMaterialAttributes!(
   MaterialAttributes
   Base Color=NamedRerouteUsage()
   Metallic=@n21
   Specular=@n21
   Roughness=Constant() {r=1}
   Emissive Color=NamedRerouteUsage()
   Normal=NamedRerouteUsage()
   World Position Offset=NamedRerouteUsage()
  ) {attributeSetTypes [
    #1
    #2
    #3
    #4
    #5
    #6
    #7
   ]}
 ]
}
